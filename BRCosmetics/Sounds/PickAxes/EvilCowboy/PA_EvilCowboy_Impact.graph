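# Sound cue for the impact sound of the EvilCowboy cosmetic pickaxe.
# A quality-level root selects one of three branches, with the last two levels sharing one, each a modulator over an equal-weight random pick from four, three or two impact waves.
# The cue primes on load, uses pickaxe impact attenuation and property sheet, plays in the impacts sound class, and references a player melee weapon concurrency set.
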SoundCue /BRCosmetics/Sounds/PickAxes/EvilCowboy/PA_EvilCowboy_Impact
root=QualityLevel_0
nodes [
 {id=Modulator_0, children [Random_0]}
 {id=Modulator_1, children [Random_1]}
 {id=Modulator_2, children [Random_2]}
 {id=QualityLevel_0, children [Modulator_0,Modulator_1,Modulator_2,Modulator_2]}
 {id=Random_0, children [WavePlayer_1,WavePlayer_2,WavePlayer_3,WavePlayer_4], properties {weights [1,1,1,1]}}
 {id=Random_1, children [WavePlayer_0,WavePlayer_5,WavePlayer_6], properties {weights [1,1,1]}}
 {id=Random_2, children [WavePlayer_7,WavePlayer_8], properties {weights [1,1]}}
 {id=WavePlayer_0, properties {soundWaveAssetPtr=/Game/Athena/Sounds/Weapons/PickAxes/EvilCowboy/Pickaxe_EvilCowboy_Impact_02}}
 {id=WavePlayer_1, properties {soundWaveAssetPtr=/Game/Athena/Sounds/Weapons/PickAxes/EvilCowboy/Pickaxe_EvilCowboy_Impact_01}}
 {id=WavePlayer_2, properties {soundWaveAssetPtr=&Pickaxe_EvilCowboy_Impact_02}}
 {id=WavePlayer_3, properties {soundWaveAssetPtr=/Game/Athena/Sounds/Weapons/PickAxes/EvilCowboy/Pickaxe_EvilCowboy_Impact_03}}
 {id=WavePlayer_4, properties {soundWaveAssetPtr=/Game/Athena/Sounds/Weapons/PickAxes/EvilCowboy/Pickaxe_EvilCowboy_Impact_04}}
 {id=WavePlayer_5, properties {soundWaveAssetPtr=&Pickaxe_EvilCowboy_Impact_03}}
 {id=WavePlayer_6, properties {soundWaveAssetPtr=&Pickaxe_EvilCowboy_Impact_04}}
 {id=WavePlayer_7, properties {soundWaveAssetPtr=&Pickaxe_EvilCowboy_Impact_01}}
 {id=WavePlayer_8, properties {soundWaveAssetPtr=&Pickaxe_EvilCowboy_Impact_03}}
]
properties {
 attenuationSettings=/Game/Sounds/Attenuation/Fort_Pickaxe_Impact_Att
 audioPropertiesSheet=/BRCosmetics/Sounds/Settings/PropertySheets/Pickaxe/APS_Cosmetic_Pickaxe_Impact
 bPrimeOnLoad=true
 maxDistance=4300
 soundClassObject=&Fort_Impacts
 volumeMultiplier=0.675
 attenuationOverrides {
  bAttenuateWithLPF=true
  distanceAlgorithm=Logarithmic
  falloffDistance=800
  lPFRadiusMax=800
  lPFRadiusMin=400
  attenuationShapeExtents {x=300}
 }
 concurrencyOverrides {bLimitToOwner=true, maxCount=2, resolutionRule=StopOldest}
 concurrencySet [&Player_Melee_Weapon_Conc]
}
