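# Sound cue for the EvilCowboy cosmetic pickaxe swing.
# A quality-level root picks among mixes of a close-attenuated random swing layer and a quieter, pitch-varied, medium-attenuated Fire Axe distant layer.
# Successive quality branches trim wave variations, with the last two sharing one mix that uses a single distant wave.
# Concurrency is overridden to stop the quietest instance, and the cue primes on load.
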SoundCue /BRCosmetics/Sounds/PickAxes/EvilCowboy/PA_EvilCowboy_Swing
root=QualityLevel_0
nodes [
 {id=Attenuation_0, children [Random_8], properties {attenuationSettings=/Game/Sounds/Attenuation/Fort_Weapon_Pickaxe_Close_Att}}
 {id=Attenuation_1, children [WavePlayer_39], properties {attenuationSettings=/Game/Sounds/Attenuation/Fort_Weapon_Pickaxe_Medium_Att}}
 {id=Attenuation_2, children [Random_6], properties {attenuationSettings=&Fort_Weapon_Pickaxe_Close_Att}}
 {id=Attenuation_3, children [Modulator_1], properties {attenuationSettings=&Fort_Weapon_Pickaxe_Medium_Att}}
 {id=Attenuation_4, children [Random_7], properties {attenuationSettings=&Fort_Weapon_Pickaxe_Close_Att}}
 {id=Attenuation_5, children [Random_2], properties {attenuationSettings=&Fort_Weapon_Pickaxe_Medium_Att}}
 {id=Mixer_0, children [Attenuation_0, Modulator_6], properties {inputVolume [1,1]}}
 {id=Mixer_1, children [Attenuation_2, Attenuation_3], properties {inputVolume [1.1,1]}}
 {id=Mixer_2, children [Attenuation_4, Modulator_5], properties {inputVolume [1.1,1]}}
 {id=Modulator_1, children [Random_3], properties {pitchMax=1.1, pitchMin=1, volumeMax=0.35, volumeMin=0.35}}
 {id=Modulator_2, children [Mixer_2]}
 {id=Modulator_5, children [Attenuation_5], properties {pitchMax=1.1, pitchMin=1, volumeMax=0.35, volumeMin=0.35}}
 {id=Modulator_6, children [Attenuation_1], properties {pitchMax=1.1, pitchMin=1, volumeMax=0.35, volumeMin=0.35}}
 {id=QualityLevel_0, children [Mixer_1,Modulator_2,Mixer_0,Mixer_0]}
 {
  id=Random_2
  children [WavePlayer_21, WavePlayer_22, WavePlayer_24, WavePlayer_25, WavePlayer_28, WavePlayer_29]
  properties {weights [1,1,1,1,1,1]}
 }
 {
  id=Random_3
  children [WavePlayer_32, WavePlayer_33, WavePlayer_34, WavePlayer_35, WavePlayer_36, WavePlayer_37]
  properties {weights [1,1,1,1,1,1]}
 }
 {id=Random_6, children [WavePlayer_23, WavePlayer_26, WavePlayer_30, WavePlayer_31, WavePlayer_38], properties {weights [1,1,1,1,1]}}
 {id=Random_7, children [WavePlayer_40, WavePlayer_41, WavePlayer_0], properties {weights [1,1,1]}}
 {id=Random_8, children [WavePlayer_43, WavePlayer_44], properties {weights [1,1]}}
 {id=WavePlayer_0, properties {soundWaveAssetPtr=&/Pickaxe_EvilCowboy_Swing_04}}
 {id=WavePlayer_21, properties {soundWaveAssetPtr=/Game/Athena/Sounds/Weapons/Fire_Axe_Swing_Distant_04}}
 {id=WavePlayer_22, properties {soundWaveAssetPtr=/Game/Athena/Sounds/Weapons/Fire_Axe_Swing_Distant_01}}
 {id=WavePlayer_23, properties {soundWaveAssetPtr=&/Pickaxe_EvilCowboy_Swing_01}}
 {id=WavePlayer_24, properties {soundWaveAssetPtr=/Game/Athena/Sounds/Weapons/Fire_Axe_Swing_Distant_02}}
 {id=WavePlayer_25, properties {soundWaveAssetPtr=/Game/Athena/Sounds/Weapons/Fire_Axe_Swing_Distant_03}}
 {id=WavePlayer_26, properties {soundWaveAssetPtr=&/Pickaxe_EvilCowboy_Swing_02}}
 {id=WavePlayer_28, properties {soundWaveAssetPtr=/Game/Athena/Sounds/Weapons/Fire_Axe_Swing_Distant_05}}
 {id=WavePlayer_29, properties {soundWaveAssetPtr=/Game/Athena/Sounds/Weapons/Fire_Axe_Swing_Distant_06}}
 {id=WavePlayer_30, properties {soundWaveAssetPtr=&Pickaxe_EvilCowboy_Swing_04}}
 {id=WavePlayer_31, properties {soundWaveAssetPtr=&/Pickaxe_EvilCowboy_Swing_05}}
 {id=WavePlayer_32, properties {soundWaveAssetPtr=&Fire_Axe_Swing_Distant_04}}
 {id=WavePlayer_33, properties {soundWaveAssetPtr=&Fire_Axe_Swing_Distant_01}}
 {id=WavePlayer_34, properties {soundWaveAssetPtr=&Fire_Axe_Swing_Distant_02}}
 {id=WavePlayer_35, properties {soundWaveAssetPtr=&Fire_Axe_Swing_Distant_03}}
 {id=WavePlayer_36, properties {soundWaveAssetPtr=&Fire_Axe_Swing_Distant_05}}
 {id=WavePlayer_37, properties {soundWaveAssetPtr=&Fire_Axe_Swing_Distant_06}}
 {id=WavePlayer_38, properties {soundWaveAssetPtr=&/Pickaxe_EvilCowboy_Swing_06}}
 {id=WavePlayer_39, properties {soundWaveAssetPtr=&Fire_Axe_Swing_Distant_01}}
 {id=WavePlayer_40, properties {soundWaveAssetPtr=&Pickaxe_EvilCowboy_Swing_01}}
 {id=WavePlayer_41, properties {soundWaveAssetPtr=&Pickaxe_EvilCowboy_Swing_02}}
 {id=WavePlayer_43, properties {soundWaveAssetPtr=&Pickaxe_EvilCowboy_Swing_01}}
 {id=WavePlayer_44, properties {soundWaveAssetPtr=&Pickaxe_EvilCowboy_Swing_02}}
]
properties {
 audioPropertiesSheet=/BRCosmetics/Sounds/Settings/PropertySheets/Pickaxe/APS_Cosmetic_Pickaxe_Swing
 bOverrideConcurrency=true
 bPrimeOnLoad=true
 maxDistance=3000
 soundClassObject=&Fort_WeaponMelee_PSM
 volumeMultiplier=0.3
 attenuationOverrides {bAttenuateWithLPF=true, bEnableListenerFocus=true, distanceAlgorithm=Logarithmic, falloffDistance=800, attenuationShapeExtents {x=300}}
 concurrencyOverrides {maxCount=3, resolutionRule=StopQuietest}
}
